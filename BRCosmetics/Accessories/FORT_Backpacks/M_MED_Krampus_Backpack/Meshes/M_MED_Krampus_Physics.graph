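# Physics asset for the Krampus medium backpack cosmetic, defining collision bodies and joints for its dynamic body and rope bones.
# The main body is jointed to the backpack root, with both using zero-length capsules that act as spheres.
# Two rope chains branch from the body, a three-bone top and a two-bone bottom, built from small capsules.
# Rope joints allow moderate swing but very little twist, while the body joint allows equal swing and twist.
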PhysicsAsset /BRCosmetics/Accessories/FORT_Backpacks/M_MED_Krampus_Backpack/Meshes/M_MED_Krampus_Physics
bodies {
 backpack_root [{shapeType=Capsule, length=0, radius=4.6496, center=<-7.025 -9.1381 -0.3169>, rotation=<rot 4.8114 -90 14.154>}]
 dyn_body_01 [{shapeType=Capsule, length=0, radius=12.0233, center=<-7.8455 0.1564 11.5755>, rotation=<rot 0 77.8831 -68.5771>}]
 dyn_rope_b_01 [{shapeType=Capsule, length=3.9415, radius=1.3737, center=<3.2547 0.0263 -0.0514>, rotation=<rot 90 0 0>}]
 dyn_rope_b_02 [{shapeType=Capsule, length=2.4826, radius=1.3591, center=<2.2382 -1.0199 1.5428>, rotation=<rot 0.6047 -113.4561 -129.272>}]
 dyn_rope_t_01 [{shapeType=Capsule, length=2.4508, radius=1.25, center=<2.2146 -0.0461 -0.1584>, rotation=<rot -0.172 104.3536 -89.2909>}]
 dyn_rope_t_02 [{shapeType=Capsule, length=3.0991, radius=1.3413, center=<2.3598 0.6155 -0.0538>, rotation=<rot 0 -78.9079 -89.4946>}]
 dyn_rope_t_03 [{shapeType=Capsule, length=1.6048, radius=1.5882, center=<2.1833 1.5314 0.5269>, rotation=<rot -24.2628 -58.2997 -105.5941>}]
}
constraints [
 {bone1Name=dyn_body_01, bone2Name=backpack_root, swing1LimitDegrees=45, swing2LimitDegrees=45, twistLimitDegrees=45}
 {bone1Name=dyn_rope_b_01, bone2Name=dyn_body_01, swing1LimitDegrees=38.1429, swing2LimitDegrees=44.6714, twistLimitDegrees=2.1429}
 {bone1Name=dyn_rope_b_02, bone2Name=dyn_rope_b_01, swing1LimitDegrees=46.7143, swing2LimitDegrees=46.7143, twistLimitDegrees=3.8571}
 {bone1Name=dyn_rope_t_01, bone2Name=dyn_body_01, swing1LimitDegrees=38, swing2LimitDegrees=45, twistLimitDegrees=2}
 {bone1Name=dyn_rope_t_02, bone2Name=dyn_rope_t_01, swing1LimitDegrees=46, swing2LimitDegrees=46, twistLimitDegrees=3}
 {bone1Name=dyn_rope_t_03, bone2Name=dyn_rope_t_02, swing1LimitDegrees=46, swing2LimitDegrees=46, twistLimitDegrees=2}
]
