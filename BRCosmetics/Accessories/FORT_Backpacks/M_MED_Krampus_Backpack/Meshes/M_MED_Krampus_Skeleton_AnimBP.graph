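# Animation Blueprint that adds physics-driven secondary motion to the Krampus backpack cosmetic's skeleton.
# Starting from the local-space ref pose, spring-driven Anim Dynamics nodes jiggle the bag, head and body bones in series.
# A RigidBody simulation in the body bone's space follows, then more Anim Dynamics nodes move both two-bone legs.
# Partial alphas temper every stage; no animations play and the event graph has no logic.
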AnimBlueprint /BRCosmetics/Accessories/FORT_Backpacks/M_MED_Krampus_Backpack/Meshes/M_MED_Krampus_Skeleton_AnimBP
target_skeleton=&/M_MED_Krampus_Skeleton
graphs [
 AnimGraph {nodes [
   n1 = Anim Dynamics - Bone: dyn_bag_b_01(ComponentPose=Local To Component(LocalPose=Local Space Ref Pose()), Alpha=0.8, ExternalForce="0,0,0") {
    angularDampingOverride=10
    angularSpringConstant=50
    bAngularSpring=true
    bGravityOverrideInSimSpace=true
    bLinearSpring=true
    gravityScale=0.4
    linearDampingOverride=50
    linearSpringConstant=20
    numSolverIterationsPostUpdate=4
    numSolverIterationsPreUpdate=16
    alphaBoolBlend {blendOption=Linear}
    boundBone {boneName=dyn_bag_b_01}
    componentAppliedLinearAccClamp=<2500 2500 2500>
    componentLinearAccScale=<1 1 1>
    componentLinearVelScale=<1 1 1>
    physicsBodyDefinitions [{
      sphereCollisionRadius=0
      boundBone {boneName=dyn_bag_b_01}
      boxExtents=<12 10 10>
      constraintSetup {angularLimitsMax=<2 1 5>, angularLimitsMin=<-1 -1 -5>, angularTarget=<0 0 0>, linearAxesMax=<1 1 5>, linearAxesMin=<-1 -1 0>}
      localJointOffset=<0 0 -1>
     }]
   }
   n2 = Anim Dynamics - Bone: dyn_head_01(ComponentPose=@n1, Alpha=0.4, ExternalForce="0,0,0") {
    angularDampingOverride=20
    angularSpringConstant=25
    bAngularSpring=true
    bGravityOverrideInSimSpace=true
    bLinearSpring=true
    gravityScale=0.5
    linearDampingOverride=20
    linearSpringConstant=50
    numSolverIterationsPostUpdate=4
    numSolverIterationsPreUpdate=16
    alphaBoolBlend {blendOption=Linear}
    boundBone {boneName=dyn_head_01}
    componentAppliedLinearAccClamp=<2500 2500 2500>
    componentLinearAccScale=<1 1 1>
    componentLinearVelScale=<1 1 1>
    physicsBodyDefinitions [{
      sphereCollisionRadius=0
      boundBone {boneName=dyn_head_01}
      boxExtents=<20 10 20>
      constraintSetup {angularLimitsMax=<45 45 20>, angularLimitsMin=<-20 -10 -20>, angularTarget=<0 0 0>, linearAxesMax=<0 1 1>, linearAxesMin=<-1 -1 -2>}
      localJointOffset=<-1 0 2>
     }]
   }
   n3 = Anim Dynamics - Bone: dyn_body_01(ComponentPose=@n2, Alpha=0.3, ExternalForce="0,0,0") {
    angularSpringConstant=50
    bAngularSpring=true
    bGravityOverrideInSimSpace=true
    bLinearSpring=true
    gravityScale=0.5
    linearSpringConstant=50
    numSolverIterationsPostUpdate=4
    numSolverIterationsPreUpdate=16
    alphaBoolBlend {blendOption=Linear}
    boundBone {boneName=dyn_body_01}
    componentAppliedLinearAccClamp=<2500 2500 2500>
    componentLinearAccScale=<1 1 1>
    componentLinearVelScale=<1 1 1>
    physicsBodyDefinitions [{
      sphereCollisionRadius=0
      boundBone {boneName=dyn_body_01}
      boxExtents=<20 10 20>
      constraintSetup {angularLimitsMax=<2 1 1>, angularLimitsMin=<-1 -1 -1>, angularTarget=<0 0 0>, linearAxesMax=<1 2 1>, linearAxesMin=<-1 0 -1>}
      localJointOffset=<0 0 -1>
     }]
   }
   n4 = RigidBody(ComponentPose=@n3, Alpha=0.5) {
    bOverrideWorldGravity=true
    simulationSpace=BaseBoneSpace
    alphaBoolBlend {blendOption=Linear}
    baseBoneRef {boneName=dyn_body_01}
    componentAppliedLinearAccClamp=<400 400 400>
    componentLinearAccScale=<1 1 1>
    componentLinearVelScale=<1 8 5>
    overrideWorldGravity {z=-980}
   }
   n5 = Anim Dynamics - Bone: dyn_leg_l_01(ComponentPose=@n4, Alpha=0.4, ExternalForce="0,0,0") {
    angularDampingOverride=50
    angularSpringConstant=50
    bAngularSpring=true
    bGravityOverrideInSimSpace=true
    bLinearSpring=true
    gravityScale=0.25
    linearDampingOverride=50
    linearSpringConstant=100
    numSolverIterationsPostUpdate=4
    numSolverIterationsPreUpdate=16
    alphaBoolBlend {blendOption=Linear}
    boundBone {boneName=dyn_leg_l_01}
    componentAppliedLinearAccClamp=<2500 2500 2500>
    componentLinearAccScale=<1 1 1>
    componentLinearVelScale=<1 1 1>
    physicsBodyDefinitions [{
      sphereCollisionRadius=0
      boundBone {boneName=dyn_leg_l_01}
      boxExtents=<25 12 25>
      constraintSetup {angularLimitsMax=<1 2 15>, angularLimitsMin=<-10 -8 -15>, angularTarget=<0 0 0>}
      localJointOffset=<0 0 -0.1>
     }]
   }
   n6 = Anim Dynamics - Bone: dyn_leg_l_02(ComponentPose=@n5, Alpha=0.2, ExternalForce="0,0,0") {
    angularDampingOverride=100
    angularSpringConstant=20
    bAngularSpring=true
    bGravityOverrideInSimSpace=true
    bLinearSpring=true
    gravityScale=0.25
    linearDampingOverride=100
    linearSpringConstant=75
    numSolverIterationsPostUpdate=4
    numSolverIterationsPreUpdate=16
    alphaBoolBlend {blendOption=Linear}
    boundBone {boneName=dyn_leg_l_02}
    componentAppliedLinearAccClamp=<2500 2500 2500>
    componentLinearAccScale=<1 1 1>
    componentLinearVelScale=<1 1 1>
    physicsBodyDefinitions [{
      sphereCollisionRadius=0
      boundBone {boneName=dyn_leg_l_02}
      boxExtents=<10 8 5>
      constraintSetup {angularLimitsMax=<15 25 15>, angularLimitsMin=<-15 -2 -15>, angularTarget=<0 0 0>, linearAxesMax=<1 0 0>, linearAxesMin=<-1 -1 -1>}
      localJointOffset=<1 0 1>
     }]
   }
   n7 = Anim Dynamics - Bone: dyn_leg_r_01(ComponentPose=@n6, Alpha=0.3, ExternalForce="0,0,0") {
    angularDampingOverride=50
    angularSpringConstant=50
    bAngularSpring=true
    bGravityOverrideInSimSpace=true
    bLinearSpring=true
    gravityScale=0.25
    linearDampingOverride=75
    linearSpringConstant=75
    numSolverIterationsPostUpdate=4
    numSolverIterationsPreUpdate=16
    alphaBoolBlend {blendOption=Linear}
    boundBone {boneName=dyn_leg_r_01}
    componentAppliedLinearAccClamp=<2500 2500 2500>
    componentLinearAccScale=<1 1 1>
    componentLinearVelScale=<1 1 1>
    physicsBodyDefinitions [{
      sphereCollisionRadius=0
      boundBone {boneName=dyn_leg_r_01}
      boxExtents=<20 6 20>
      constraintSetup {angularLimitsMax=<1 1 1>, angularLimitsMin=<-1 -1 -1>, angularTarget=<0 0 0>, linearAxesMax=<1 0 0>, linearAxesMin=<-1 -1.5 -1>}
      localJointOffset=<0 0 -1>
     }]
   }
   n8 = Anim Dynamics - Bone: dyn_leg_r_02(ComponentPose=@n7, Alpha=0.5, ExternalForce="0,0,0") {
    angularDampingOverride=100
    angularSpringConstant=50
    bAngularSpring=true
    bGravityOverrideInSimSpace=true
    bLinearSpring=true
    gravityScale=0.25
    linearDampingOverride=100
    linearSpringConstant=75
    numSolverIterationsPostUpdate=4
    numSolverIterationsPreUpdate=16
    alphaBoolBlend {blendOption=Linear}
    boundBone {boneName=dyn_leg_r_02}
    componentAppliedLinearAccClamp=<2500 2500 2500>
    componentLinearAccScale=<1 1 1>
    componentLinearVelScale=<1 1 1>
    physicsBodyDefinitions [{
      sphereCollisionRadius=0
      boundBone {boneName=dyn_leg_r_02}
      boxExtents=<10 8 5>
      constraintSetup {angularLimitsMax=<2 2 2>, angularLimitsMin=<-2 -2 -2>, angularTarget=<0 0 0>, linearAxesMax=<1 0 0>, linearAxesMin=<-1 -1.5 -1>}
      localJointOffset=<1 0 1>
     }]
   }
   Output Pose(Result=Component To Local(ComponentPose=@n8)) {layerGroup=None}
  ]}
 EventGraph {nodes [
   TryGetPawnOwner()
   Event BlueprintUpdateAnimation()
  ]}
]
variables []
